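# Niagara idle effect for the female Barium pickaxe, producing a short blue glint from one-shot bursts.
# DualColorGlint and InitialFlash spawn mesh particles whose curve-driven dynamic material parameters and alpha fades animate the look, with DualColorGlint also writing alternate color and parameter attributes.
# Glint_Start spawns two rotating sprites using system glint size and color, with hue shifting, camera offset and a curve-driven position that climbs along Z.
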NiagaraSystem /BRCosmetics/Weapons/FORT_Melee/Pickaxe_Barium_Female/FX/NS_Pickaxe_Barium_Female_Idle
dependencies {
 usedDataInterfaces [&NiagaraDataInterfaceCurve]
 usedDynamicInputs [&Multiply_Vector2]
 usedModules [
  &self:SystemSpawnScript.SystemScriptSource.SystemScriptGraph.NiagaraNodeAssignment_0.SetVariables_#1
  &self:SystemSpawnScript.SystemScriptSource.SystemScriptGraph.NiagaraNodeAssignment_2.SetVariables_#2
  &SystemState
 ]
}
emitters {
 DualColorGlint {
  emitterUpdateScript [
   &EmitterState()
   &SpawnBurst_Instantaneous(Spawn Count=1, Use Spawn Probability=false)
  ]
  particleSpawnScript [
   &InitializeParticle(
    Color=<rgba 0 1.1667 10 0.2>
    Lifetime=0.3
    Position=<0 0 0.75>
    Position Mode=::NewEnumerator1
    UsePositionOffset=false
    Write Material Random=true
   )
   /SetVariables(Particles.ColorAlt:LinearColor=<rgba 0 23.3333 200 1>) {enabled=false, isSetParameters=true, module=SetVariables_#3}
  ]
  particleUpdateScript [
   &ParticleState()
   &DynamicMaterialParameters(
    Index 0 Param 1=&FloatFromCurve(
     FloatCurve=DataInterface() {propertyValues {Curve {keys [
         {arriveTangent=0.0136, interpMode=RCIM_Cubic, time=0, value=0}
         {arriveTangent=1.75, interpMode=RCIM_Cubic, leaveTangent=1.75, time=0.2, value=0.15}
         {arriveTangent=1.0625, interpMode=RCIM_Cubic, leaveTangent=1.0625, time=0.5, value=0.7}
         {interpMode=RCIM_Cubic, leaveTangent=1.0398, time=1, value=1}
        ]}}}
     Scale Curve=1.4
    )
    Index 0 Param 2=&FloatFromCurve(FloatCurve=DataInterface() {propertyValues {Curve {keys [<at 0.2 0.25>,<at 0.3 1.8>,<at 1 0>]}}})
    Index 0 Param 3=&FloatFromCurve(
     FloatCurve=DataInterface() {propertyValues {Curve {keys [
         {interpMode=RCIM_Cubic, tangentMode=RCTM_User, time=0.4, value=0}
         {interpMode=RCIM_Cubic, tangentMode=RCTM_User, time=0.5, value=1}
         <at 0.6 1>
         {tangentMode=RCTM_User, time=0.7, value=0}
        ]}}}
     Scale Curve=0
    )
    Module.Param0WriteEnabled=true
    Module.Param1WriteEnabled=true
    Module.Param2WriteEnabled=true
    Module.Param3WriteEnabled=true
   )
   /SetVariables(Particles.DynaMatAlt:Vector4f=&MakeVector4(
     X=&FloatFromCurve(
      FloatCurve=DataInterface() {propertyValues {Curve {keys [
          {arriveTangent=2.3402, interpMode=RCIM_Cubic, leaveTangent=2.3402, tangentMode=RCTM_User, time=0, value=0}
          {interpMode=RCIM_Cubic, tangentMode=RCTM_User, time=1, value=1}
         ]}}}
      Scale Curve=1.4
     )
     Y=&FloatFromCurve(FloatCurve=DataInterface() {propertyValues {Curve {keys [<at 0.2 2>,<at 0.58 0>]}}})
     Z=&FloatFromCurve(
      FloatCurve=DataInterface() {propertyValues {Curve {keys [<at 0.6244 0 RCIM_Cubic>,<at 0.6744 1 RCIM_Cubic>,<at 1 0 RCIM_Cubic>]}}}
      Scale Curve=0
     )
     W=&FloatFromCurve(FloatCurve=DataInterface() {propertyValues {Curve {keys [<at 0 0>,<at 0.3 0>,<at 0.6 1>]}}})
    )) {isSetParameters=true, module=SetVariables_#4}
   &ScaleColor(
    Scale Alpha=&FloatFromCurve(FloatCurve=DataInterface() {propertyValues {Curve {keys [<at 0.2 1>,<at 1 0>]}}})
    ScaleA=true
    ScaleRGB=true
   )
   /SetVariables(Particles.ColorAltUpdate:LinearColor=&MakeLinearColorFromVectorAndFloat(
     "Vector (RGB)"=&MakeVectorFromLinearColorRGB(Color=<rgba 0 2.3333 20 1>)
     "Float (Alpha)"=&FloatFromCurve(FloatCurve=DataInterface() {propertyValues {Curve {keys [<at 0.5 1>,<at 1 0>]}}})
    )) {isSetParameters=true, module=SetVariables_#5}
  ]
  rendererClasses [&NiagaraMeshRendererProperties]
  renderers [
   {rendererClass=&NiagaraMeshRendererProperties, rendererIndex=0}
   {rendererClass=&NiagaraMeshRendererProperties, rendererIndex=1}
  ]
 }
 Glint_Start {
  emitterUpdateScript [
   &EmitterState()
   &SpawnBurst_Instantaneous(Spawn Count=2, Use Spawn Probability=false)
  ]
  particleSpawnScript [
   /Niagara/Modules/Spawn/Initialization/InitializeParticle(
    Lifetime=0.25
    Module.Write Color=true
    Module.Write Lifetime=true
    Module.Write Mass=true
    Module.Write Material Random=true
    Module.Write Position=true
    Module.Write Scale=false
    Module.Write SpriteRotation=true
    Module.Write SpriteSize=true
    Position:Vector3f=$System.GlintLocation
    Sprite Rotation=&RandomRangeFloat(Minimum=0, Maximum=350, Evaluation Type=::NewEnumerator1)
    Sprite Size=&RandomRangeVector2D(Minimum=<56.25 33.75>, Maximum=<78.75 56.25>, Evaluation Type=::NewEnumerator1)
   )
   /SetVariables(Particles.SpriteSize=$System.SystemGlintSize) {isSetParameters=true}
  ]
  particleUpdateScript [
   &ParticleState()
   &ScaleSpriteSize(Scale Factor=&Vector2DFromFloat(Value=&FloatFromCurve(
      FloatCurve=DataInterface() {propertyValues {Curve {keys [
          <at 0 0 RCIM_Cubic>
          {arriveTangent=1.8529, interpMode=RCIM_Cubic, leaveTangent=1.8529, time=0.0545, value=0.9401}
          {arriveTangent=-0.9943, interpMode=RCIM_Cubic, leaveTangent=-0.9943, time=0.4374, value=0.8104}
          <at 1 0>
         ]}}}
      Scale Curve=2
     )))
   &CameraOffset(Camera Offset Amount=&FloatFromCurve(FloatCurve=DataInterface() {propertyValues {Curve {keys [<at 0 15>,<at 1 20>]}}}, Scale Curve=2))
   &Color(Color=&HueShiftLinearColor(
     Hue=&FloatFromCurve(FloatCurve=DataInterface() {propertyValues {Curve {keys [{tangentMode=RCTM_User, time=0, value=0.4}, {tangentMode=RCTM_User, time=1, value=1.5}]}}})
     Saturation=1
     Value=&RandomRangeFloat(Minimum=0.2, Maximum=0.5)
     Linear Color=$System.GlintColor
    ))
   /SetVariables(Particles.Position=&MakeVector(
     X=-1
     Y=0
     Z=&Add_Float(
      A=30
      B=&FloatFromCurve(
       FloatCurve=DataInterface() {propertyValues {Curve {keys [
           {arriveTangent=0.0136, interpMode=RCIM_Cubic, time=0, value=0}
           {arriveTangent=1.75, interpMode=RCIM_Cubic, leaveTangent=1.75, time=0.2, value=0.15}
           {arriveTangent=1.0625, interpMode=RCIM_Cubic, leaveTangent=1.0625, time=0.5, value=0.7}
           {interpMode=RCIM_Cubic, leaveTangent=1.0398, time=1, value=1}
          ]}}}
       Scale Curve=30
      )
     )
    )) {isSetParameters=true}
   &SpriteRotationRate(Rotation Rate=100)
  ]
  rendererClasses [&NiagaraSpriteRendererProperties]
  renderers [{rendererClass=&NiagaraSpriteRendererProperties}]
 }
 InitialFlash {
  emitterUpdateScript [
   &EmitterState()
   &SpawnBurst_Instantaneous(Spawn Count=1, Use Spawn Probability=false)
  ]
  particleSpawnScript [
   /Niagara/Modules/Spawn/Initialization/V2/InitializeParticle(
    Color=<rgba 0.5 1.025 5 1>
    Lifetime=0.3
    Position=<0 0 0.75>
    Position Mode=::NewEnumerator1
    UsePositionOffset=false
    Write Material Random=true
   )
   /SetVariables(Particles.ColorAlt:LinearColor=<rgba 0 23.3333 200 1>) {enabled=false, isSetParameters=true, module=SetVariables_#3}
  ]
  particleUpdateScript [
   &ParticleState()
   &DynamicMaterialParameters(
    Index 0 Param 1=&FloatFromCurve(
     FloatCurve=DataInterface() {propertyValues {Curve {keys [
         {arriveTangent=0.0136, interpMode=RCIM_Cubic, time=0, value=0}
         {arriveTangent=1.75, interpMode=RCIM_Cubic, leaveTangent=1.75, time=0.2, value=0.15}
         {arriveTangent=1.0625, interpMode=RCIM_Cubic, leaveTangent=1.0625, time=0.5, value=0.7}
         {interpMode=RCIM_Cubic, leaveTangent=1.0398, time=1, value=1}
        ]}}}
     Scale Curve=1.4
    )
    Index 0 Param 2=&FloatFromCurve(FloatCurve=DataInterface() {propertyValues {Curve {keys [<at 0.2 0.25>,<at 0.3 1.8>,<at 1 0>]}}})
    Index 0 Param 3=&FloatFromCurve(
     FloatCurve=DataInterface() {propertyValues {Curve {keys [
         {interpMode=RCIM_Cubic, tangentMode=RCTM_User, time=0, value=0}
         {arriveTangent=2.4495, interpMode=RCIM_Cubic, leaveTangent=2.4495, tangentMode=RCTM_User, time=0.5, value=1}
         <at 0.6 1>
         <at 0.649 0.4515>
         {tangentMode=RCTM_User, time=1, value=0}
        ]}}}
     Scale Curve=0.2
    )
    Index 0 Param 4=&FloatFromCurve(
     FloatCurve=DataInterface() {propertyValues {Curve {keys [{interpMode=RCIM_Cubic, tangentMode=RCTM_User, time=0, value=0}, {interpMode=RCIM_Cubic, tangentMode=RCTM_User, time=1, value=1}]}}}
     Scale Curve=0.1
    )
    Module.Param0WriteEnabled=true
    Module.Param1WriteEnabled=true
    Module.Param2WriteEnabled=true
    Module.Param3WriteEnabled=true
   )
   /SetVariables(Particles.DynaMatAlt:Vector4f=&MakeVector4(
     X=&FloatFromCurve(
      FloatCurve=DataInterface() {propertyValues {Curve {keys [
          {arriveTangent=2.3402, interpMode=RCIM_Cubic, leaveTangent=2.3402, tangentMode=RCTM_User, time=0, value=0}
          {interpMode=RCIM_Cubic, tangentMode=RCTM_User, time=1, value=1}
         ]}}}
      Scale Curve=1.4
     )
     Y=&FloatFromCurve(FloatCurve=DataInterface() {propertyValues {Curve {keys [<at 0.2 2.5>,<at 0.58 0>]}}})
     Z=&FloatFromCurve(
      FloatCurve=DataInterface() {propertyValues {Curve {keys [<at 0.6244 0 RCIM_Cubic>,<at 0.6744 1 RCIM_Cubic>,<at 1 0 RCIM_Cubic>]}}}
      Scale Curve=0
     )
     W=&FloatFromCurve(FloatCurve=DataInterface() {propertyValues {Curve {keys [<at 0 0>,<at 0.3 0>,<at 0.6 1>]}}})
    )) {enabled=false, isSetParameters=true, module=SetVariables_#4}
   &ScaleColor(
    Scale Alpha=&FloatFromCurve(FloatCurve=DataInterface() {propertyValues {Curve {keys [<at 0.2 1>,<at 1 0>]}}})
    ScaleA=true
    ScaleRGB=true
   )
   /SetVariables(Particles.ColorAltUpdate:LinearColor=&MakeLinearColorFromVectorAndFloat(
     "Vector (RGB)"=&MakeVectorFromLinearColorRGB(Color=<rgba 0 2.3333 20 1>)
     "Float (Alpha)"=&FloatFromCurve(FloatCurve=DataInterface() {propertyValues {Curve {keys [<at 0.5 1>,<at 1 0>]}}})
    )) {enabled=false, isSetParameters=true, module=SetVariables_#5}
  ]
  rendererClasses [&NiagaraMeshRendererProperties]
  renderers [{rendererClass=&NiagaraMeshRendererProperties, rendererIndex=0}]
 }
}
summary {
 systemName=NS_Pickaxe_Barium_Female_Idle
 emitters [
  {emitterName=DualColorGlint, rendererClasses [&NiagaraMeshRendererProperties]}
  {emitterName=InitialFlash, rendererClasses [&NiagaraMeshRendererProperties]}
  {emitterName=Glint_Start, rendererClasses [&NiagaraSpriteRendererProperties]}
 ]
 userVariables [User.Skeletal Mesh {
   type=NiagaraDataInterfaceSkeletalMesh
   defaultValue {
    struct=NiagaraExt_VariableValue_DataInterface
    value {dataInterface=&self:NiagaraDataInterfaceSkeletalMesh_0}
   }
  }]
}
